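# Default Lit skeletal-mesh material that renders the Beak T2 Ice car body's headlights as glowing, layered ice.
# Samples a tiled water normal map at several BumpOffset depths, plus a parallax, optionally panning background texture, to fake internal depth.
# Camera-facing Fresnel blends these layers with inner and outer ice tints into a brightness-scaled, clamped emissive.
# Named reroutes also carry a tinted base color and an intensity-blended normal, alongside metallic and roughness parameters.
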Material /VehicleCosmetics/Bodies/BeakT2Ice/Materials/M_Beak_T2_Headlights_Ice
graph {
 properties {bAutomaticallySetUsageInEditor=false, bUsedWithSkeletalMesh=true, usedShadingModels=Default Lit}
 nodes [
  n1 = Constant3Vector() {constant=<rgba 0 0 1 1>}
  n2 = NamedRerouteUsage()
  n3 = ScalarParameter() {defaultValue=0.2, group=Ice, parameterName=NormalIntensity, sortPriority=4}
  NamedRerouteDeclaration(None=Clamp(None=LinearInterpolate(A=@n1, B=@n2, Alpha=@n3))) {name=Normal, nodeColor=<rgba 0 0.0625 1 1>}
  n4 = Multiply(A=TextureCoordinate(), B=ScalarParameter() {defaultValue=1, group=Ice, parameterName=Tiling, sortPriority=3})
  n5 = TextureObjectParameter() {
   group=Ice
   parameterName=Normal_LinearSpace
   samplerType=SAMPLERTYPE_LinearColor
   sortPriority=-2
   texture=/VehicleCosmetics/Wheels/_SharedTextures/FX/Lux_Water01_N
  }
  n6 = TextureSample(UVs=BumpOffset(Coordinate=@n4, Height=Constant() {r=-0.15}), Tex=@n5) {
   samplerSource=SSM_Wrap_WorldGroupSettings
   samplerType=SAMPLERTYPE_LinearColor
   texture=&Lux_Water01_N
  }
  n7 = Multiply(A=AppendVector(A=@n6.R, B=@n6.R)) {constB=2}
  n8 = Desaturation(None=DeriveNormalZ(InXY=Subtract(A=@n7))) {luminanceFactors=<rgba 0.3 0.59 0.11 1>}
  n9 = Clamp(None=ConstantBiasScale(None=@n8) {bias=0.2})
  n10 = CameraVectorWS()
  n11 = TextureSample(UVs=@n4, Tex=@n5) {
   samplerSource=SSM_Wrap_WorldGroupSettings
   samplerType=SAMPLERTYPE_LinearColor
   texture=&Lux_Water01_N
  }
  n12 = Multiply(A=AppendVector(A=@n11.R, B=@n11.R)) {constB=2}
  n13 = NamedRerouteDeclaration(None=DeriveNormalZ(InXY=Subtract(A=@n12))) {name=N1, nodeColor=<rgba 1 0.2578 0 1>}
  n14 = Transform(None=@n13)
  n15 = Saturate(None=OneMinus(None=DotProduct(A=@n10, B=@n14)))
  n16 = Power(Base=@n15)
  n17 = LinearInterpolate(A=Clamp(None=ConstantBiasScale(None=@n9) {bias=-0.1}), Alpha=@n16) {constB=0}
  n18 = TextureSampleParameter2D() {group=Ice, parameterName=Diffuse, sortPriority=-3, texture=/VehicleCosmetics/Wheels/_SharedTextures/FX/Distort_RGB_T}
  n19 = TextureSample(UVs=BumpOffset(Coordinate=@n4, Height=Constant() {r=-0.2}), Tex=@n5) {
   samplerSource=SSM_Wrap_WorldGroupSettings
   samplerType=SAMPLERTYPE_LinearColor
   texture=&Lux_Water01_N
  }
  n20 = Subtract(A=Multiply(A=@n8, B=@n19.A), B=@n18.G)
  n21 = Power(Base=Add(A=Add(A=@n17, B=@n18.G), B=Clamp(None=@n20)))
  n22 = Desaturation(None=@n18.RGB) {luminanceFactors=<rgba 0.3 0.59 0.11 1>}
  n23 = LinearInterpolate(A=Desaturation(None=@n13) {luminanceFactors=<rgba 0.3 0.56 0.11 1>}, B=@n22) {constAlpha=0.75}
  n24 = AppendVector(A=@n19.R, B=@n19.R)
  n25 = DeriveNormalZ(InXY=Subtract(A=Multiply(A=@n24) {constB=2}))
  n26 = BumpOffset(Coordinate=@n4, Height=Constant() {r=-3})
  n27 = Panner(Coordinate=@n26) {speedY=-0.001}
  n28 = Multiply(
   A=Multiply(A=LinearInterpolate(Alpha=@n16) {constA=3, constB=0}, B=Desaturation(None=@n25) {luminanceFactors=<rgba 0.3 0.59 0.11 1>})
   B=TextureSampleParameter2D(UVs=StaticSwitchParameter(True=@n27, False=@n26) {defaultValue=true, group=Ice, parameterName=Panning BG, sortPriority=2}) {group=Ice, parameterName=RGB, sortPriority=-1, texture=/VehicleCosmetics/Wheels/_SharedTextures/FX/T_IceRedux_2RGB}.R
  )
  n29 = Clamp(None=ConstantBiasScale(None=OneMinus(None=@n23)) {bias=0.5, scale=1})
  n30 = Add(A=Add(A=Multiply(B=@n21) {constA=1.2}, B=Multiply(A=@n23, B=@n18.R)), B=Multiply(B=Multiply(A=@n28, B=@n29)) {constA=1.2})
  n31 = VectorParameter() {group=Ice, parameterName=InnerIce, sortPriority=0, defaultValue=<rgba 0.2335 0.5857 0.9719 1>}
  n32 = Clamp(None=ConstantBiasScale(None=Multiply(A=@n11.A, B=@n18.G)) {bias=0.1, scale=1})
  n33 = Power(Base=@n15) {constExponent=1}
  n34 = Add(A=@n33, B=@n9)
  n35 = VectorParameter() {group=Ice, parameterName=OuterIce, sortPriority=1, defaultValue=<rgba 0.3098 0.7729 0.9831 1>}
  n36 = Add(A=LinearInterpolate(A=Multiply(A=@n30, B=@n31.RGB), B=Add(A=@n32, B=@n31.RGB), Alpha=Clamp(None=@n34)), B=Multiply(A=@n33, B=@n35.RGB))
  n37 = ScalarParameter() {defaultValue=1.5, group=Ice, parameterName=EmissiveBrightness, sortPriority=5}
  NamedRerouteDeclaration(None=Clamp(None=Multiply(A=@n36, B=@n37))) {name=Emissive, nodeColor=<rgba 0.3828 1 0 1>}
  n38 = Multiply(A=@n35.RGB, B=@n32)
  n39 = LinearInterpolate(A=@n11.A, B=@n22)
  NamedRerouteDeclaration(None=Multiply(A=Multiply(A=@n38, B=@n39)) {constB=1.2}) {name=Base Color, nodeColor=<rgba 1 0 0.7031 1>}
  NamedRerouteUsage()
  NamedRerouteUsage()
  NamedRerouteUsage()
  ScalarParameter() {group=Ice, parameterName=Metallic, sortPriority=6}
  ScalarParameter() {defaultValue=0.1, group=Ice, parameterName=Roughness, sortPriority=7}
 ]
}
