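# Physics asset for the BeakT2Ice vehicle body cosmetic mesh, in its LOD0 BR variant, giving the vehicle a simplified collision and simulation shape.
# It holds a single body on the frame bone, built from one large rotated capsule offset from the bone origin.
# No constraints are defined, so the body acts as one rigid piece with no articulated parts.
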PhysicsAsset /VehicleCosmetics/Bodies/BeakT2Ice/Meshes/SM_BeakT2Ice_LOD0_Physics_BR
bodies {frame [{shapeType=Capsule, length=247.9842, radius=115.0131, center=<41.4331 -53.2058 0.0063>, rotation=<rot 0 92.9157 -89.9966>}]}
constraints []
